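; Parent burst gameplay cue for CoffeePuff melee hits that plays an impact sound at the hit location.
; On burst, it breaks the hit result from the cue's effect context and casts the hit actor to FortPawn to choose player-hit or world-hit audio.
; Each choice has first- and third-person variants held in variables, selected by whether the cue's target pawn is locally viewed.
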(fn OnBurst(EitherVFX-preferthisversion)
  (params (MyTarget :type Actor%20Object%20Reference) (Parameters :type Gameplay%20Cue%20Parameters%20Structure%20%28by%20ref%29) (ParticleComponents :type Array%20of%20FXSystem%20Component%20Object%20References) (AudioComponents :type Array%20of%20Audio%20Component%20Object%20References) (BurstCameraShakeInstance :type Camera%20Shake%20Base%20Object%20Reference) (BurstDecalInstance :type Decal%20Actor%20Object%20Reference))
  (bind (_normalizedmagnitude _rawmagnitude _effectcontext) (Ability|GameplayCue|BreakGameplayCueParameters Parameters))
  (bind _returnvalue (Ability|EffectContext|GetHitResult _effectcontext))
  (bind (_bblockinghit _binitialoverlap _time _distance _location _impactpoint _normal _impactnormal _physmat _hitactor) (Collision|BreakHitResult _returnvalue))
  (bind _asfort_pawn (Utilities|Casting|CastToFortPawn MyTarget))
  (bind _asfort_pawn_2 (Utilities|Casting|CastToFortPawn _hitactor)
    (:then
      (Audio|PlaySoundatLocation _asfort_pawn_2 (select (Pawn|IsLocallyViewed _asfort_pawn) (Variables|Default|GetPlayerHitSound1P) (Variables|Default|GetPlayerHitSound3P)) _location))
    (:CastFailed
      (Audio|PlaySoundatLocation _asfort_pawn (select (Pawn|IsLocallyViewed _asfort_pawn) (Variables|Default|GetWorldHitSound1P) (Variables|Default|GetWorldHitSound3P)) _location)))
  (return))
